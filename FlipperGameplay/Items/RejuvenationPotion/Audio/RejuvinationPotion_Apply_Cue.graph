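# Sound cue that plays when a Rejuvenation Potion is applied, under the consumables sound class with consumable-use attenuation.
# A mixer layers a quieter, fixed-pitch generic health-pickup sound with the potion's own apply sound, which starts after a short fixed delay at a slightly randomized, boosted volume.
# A per-owner concurrency override allows one instance at a time, stopping the oldest.
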SoundCue /FlipperGameplay/Items/RejuvenationPotion/Audio/RejuvinationPotion_Apply_Cue
root=Mixer_0
nodes [
 {id=Delay_0, children [Modulator_0], properties {delayMax=0.25, delayMin=0.25}}
 {id=Mixer_0, children [Delay_0,Modulator_1], properties {inputVolume [1,1]}}
 {id=Modulator_0, children [WavePlayer_1], properties {volumeMax=1.5, volumeMin=1.3}}
 {id=Modulator_1, children [WavePlayer_0], properties {pitchMax=1, pitchMin=1, volumeMax=0.38, volumeMin=0.38}}
 {id=WavePlayer_0, properties {soundWaveAssetPtr=/Game/Sounds/Fort_Gadgets/Unassigned/new_health_get}}
 {id=WavePlayer_1, properties {soundWaveAssetPtr=/FlipperGameplay/Items/RejuvenationPotion/Audio/wav/Item_Consumables_Rejuv_Potion_Apply}}
]
properties {
 attenuationSettings=&Fort_Consumable_Use_Athena_Att
 bOverrideConcurrency=true
 maxDistance=2250
 soundClassObject=&Fort_Consumables
 volumeMultiplier=1
 concurrencyOverrides {bLimitToOwner=true, maxCount=1, resolutionRule=StopOldest}
}
