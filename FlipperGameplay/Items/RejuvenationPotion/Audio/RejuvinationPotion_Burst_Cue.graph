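# Sound cue for the burst sound of the Rejuvenation Potion consumable.
# A single wave player for the burst wave feeds a modulator at the root.
# It uses the consumables sound class and Athena consumable-use attenuation, with a concurrency override that caps simultaneous instances at a small count by stopping the oldest.
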SoundCue /FlipperGameplay/Items/RejuvenationPotion/Audio/RejuvinationPotion_Burst_Cue
root=Modulator_0
nodes [
 {id=Modulator_0, children [WavePlayer_1]}
 {id=WavePlayer_1, properties {soundWaveAssetPtr=/FlipperGameplay/Items/RejuvenationPotion/Audio/wav/Item_Consumables_Rejuv_Potion_Burst}}
]
properties {
 attenuationSettings=&Fort_Consumable_Use_Athena_Att
 bOverrideConcurrency=true
 maxDistance=2250
 soundClassObject=&Fort_Consumables
 volumeMultiplier=1
 concurrencyOverrides {maxCount=2, resolutionRule=StopOldest}
}
