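# Sound cue for drinking the Rejuvenation Potion consumable in FlipperGameplay.
# A single wave player feeds a Modulator that applies a slight random volume boost to each play.
# It routes through the Fort consumables sound class with overridden attenuation, adding occlusion and distance falloff.
# Overridden concurrency caps simultaneous instances and stops the oldest first.
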SoundCue /FlipperGameplay/Items/RejuvenationPotion/Audio/RejuvinationPotion_Drink_Cue
root=Modulator_0
nodes [
 {id=Modulator_0, children [WavePlayer_0], properties {volumeMax=1.3, volumeMin=1.1}}
 {id=WavePlayer_0, properties {soundWaveAssetPtr=/FlipperGameplay/Items/RejuvenationPotion/Audio/wav/Item_Consumables_Rejuv_Potion_Drink}}
]
properties {
 attenuationSettings=&Fort_Consumable_Use_Athena_Att
 bOverrideAttenuation=true
 bOverrideConcurrency=true
 maxDistance=1600
 soundClassObject=&Fort_Consumables
 volumeMultiplier=1
 attenuationOverrides {bEnableOcclusion=true, falloffDistance=1200, occlusionTraceChannel=ECC_GameTraceChannel11, occlusionVolumeAttenuation=0.8}
 concurrencyOverrides {maxCount=2, resolutionRule=StopOldest}
}
