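# Sound cue for the end of the Rejuvenation Potion drink loop in FlipperGameplay.
# A single wave player feeds the potion's end wave straight to the output, with no randomization or modulation.
# It uses the Athena consumable-use attenuation and the consumables sound class, with a concurrency override that stops the oldest instance when the limit is reached.
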SoundCue /FlipperGameplay/Items/RejuvenationPotion/Audio/RejuvinationPotion_Drink_lp_End_Cue
root=WavePlayer_0
nodes [{id=WavePlayer_0, properties {soundWaveAssetPtr=/FlipperGameplay/Items/RejuvenationPotion/Audio/wav/Item_Consumables_Rejuv_Potion_End}}]
properties {
 attenuationSettings=&Fort_Consumable_Use_Athena_Att
 bOverrideConcurrency=true
 maxDistance=2250
 soundClassObject=&Fort_Consumables
 volumeMultiplier=1
 concurrencyOverrides {maxCount=4, resolutionRule=StopOldest}
}
