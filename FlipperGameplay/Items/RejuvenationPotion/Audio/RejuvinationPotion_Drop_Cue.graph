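# Sound cue that plays when a Rejuvenation Potion item is dropped.
# A single wave player feeds the potion drop wav into a root modulator that randomizes volume slightly upward on each play.
# It uses the shared weapon-drop attenuation and the loot-collect sound class.
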SoundCue /FlipperGameplay/Items/RejuvenationPotion/Audio/RejuvinationPotion_Drop_Cue
root=Modulator_0
nodes [
 {id=Modulator_0, children [WavePlayer_1], properties {volumeMax=1.3, volumeMin=1}}
 {id=WavePlayer_1, properties {soundWaveAssetPtr=/FlipperGameplay/Items/RejuvenationPotion/Audio/wav/Item_Consumables_Rejuv_Potion_Drop}}
]
properties {
 attenuationSettings=&Fort_Weapon_Drop_Att
 maxDistance=3000
 soundClassObject=&Fort_Loot_Collect
 volumeMultiplier=1
}
