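# Sound cue that plays when the Rejuvenation Potion item is equipped.
# A single wave player plays the potion equip wave through a Modulator that randomizes volume slightly upward from base level.
# It uses the weapon reload attenuation and the loot collect sound class, with overridden concurrency that caps simultaneous instances.
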SoundCue /FlipperGameplay/Items/RejuvenationPotion/Audio/RejuvinationPotion_Equip_Cue
root=Modulator_0
nodes [
 {id=Modulator_0, children [WavePlayer_0], properties {volumeMax=1.3, volumeMin=1}}
 {id=WavePlayer_0, properties {soundWaveAssetPtr=/FlipperGameplay/Items/RejuvenationPotion/Audio/wav/Item_Consumables_Rejuv_Potion_Equip}}
]
properties {
 attenuationSettings=&Fort_Weapon_Reload_Att
 bOverrideConcurrency=true
 maxDistance=2250
 soundClassObject=&Fort_Loot_Collect
 volumeMultiplier=1
 concurrencyOverrides {maxCount=3}
}
