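# Sound cue for picking up the Rejuvenation Potion item.
# A single pickup wave plays through a Modulator node left at its usual settings.
# The cue uses the shared weapon pickup attenuation and the gadget items sound class at a slightly reduced volume.
# Its own concurrency override blocks new instances once the limit is reached rather than stopping playing ones.
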SoundCue /FlipperGameplay/Items/RejuvenationPotion/Audio/RejuvinationPotion_Pickup_Cue
root=Modulator_0
nodes [
 {id=Modulator_0, children [WavePlayer_0]}
 {id=WavePlayer_0, properties {soundWaveAssetPtr=/FlipperGameplay/Items/RejuvenationPotion/Audio/wav/Item_Consumables_Rejuv_Potion_Pickup}}
]
properties {
 attenuationSettings=&Fort_Weapon_Pickup_Att
 bOverrideConcurrency=true
 maxDistance=3500
 soundClassObject=&Fort_Gadgets_Items
 volumeMultiplier=0.8
 concurrencyOverrides {resolutionRule=PreventNew}
}
